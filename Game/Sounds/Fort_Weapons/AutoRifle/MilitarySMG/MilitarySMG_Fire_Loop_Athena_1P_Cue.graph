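# Looping first-person fire sound cue for the Military SMG in Athena.
# A single looping wave player plays the close fire-loop sample through a modulator that slightly varies volume while holding pitch fixed.
# It uses first-person weapon attenuation, a first-person ranged-fire sound class and first-person gunfire concurrency, and has the debug flag enabled.
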SoundCue /Game/Sounds/Fort_Weapons/AutoRifle/MilitarySMG/MilitarySMG_Fire_Loop_Athena_1P_Cue
root=Modulator_1
nodes [
 {id=Modulator_1, children [WavePlayer_0], properties {pitchMax=1, pitchMin=1, volumeMax=0.8, volumeMin=0.75}}
 {id=WavePlayer_0, properties {bLooping=true, soundWaveAssetPtr=&/MilitarySMG_Close_FireLoop_01}}
]
properties {
 attenuationSettings=&Fort_Weapon_Projectile_FirstPerson_Athena_Att
 bDebug=true
 maxDistance=2097152
 soundClassObject=&Fort_WeaponRanged_Fire_1P_PSM
 concurrencySet [&Weapon_Gunfire_FirstPerson_Conc]
}
