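# Sound cue for the military SMG's low-ammo weapon sound, built on one of the gun's close fire-stop waves.
# A looping wave player feeds a modulator that randomizes pitch slightly upward and holds volume steady.
# The cue boosts overall volume, routes to the degraded low-ammo ranged weapon sound class, and overrides concurrency to prevent new instances at the limit.
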SoundCue /Game/Sounds/Fort_Weapons/AutoRifle/MilitarySMG/MilitarySMG_LowAmmo_Cue
root=Modulator_0
nodes [
 {id=Modulator_0, children [WavePlayer_70], properties {pitchMax=1.1, pitchMin=1, volumeMax=1, volumeMin=1}}
 {
  id=WavePlayer_70
  properties {bLooping=true, soundWaveAssetPtr=&/MilitarySMG_Close_FireStop_03}
 }
]
properties {
 bOverrideConcurrency=true
 maxDistance=2097152
 soundClassObject=/Game/Sounds/FortSoundClassesAndModes/Fort_WeaponRanged_LowAmmo_Degraded_PSM
 volumeMultiplier=1.9069
 concurrencyOverrides {resolutionRule=PreventNew}
}
